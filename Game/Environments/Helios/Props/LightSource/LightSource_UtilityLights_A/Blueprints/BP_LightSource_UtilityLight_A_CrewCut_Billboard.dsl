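; Blueprint for the CrewCut Billboard variant of the Helios UtilityLight A light source prop, tinting itself from building texture data.
; After the parent construction script runs, it checks for building texture data and passes the first entry's tint color to SetColoronSecondaryMesh.
; It then sets material slot 0 of the static mesh to a dynamic instance from the tech art global library.
; BeginPlay, overlap and tick events are disabled stubs.
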(event EventBeginPlay
  (|CallParentFunction) ; This node is disabled and will not be called.\nDrag off pins to build functionality.
)

(event Collision|EventActorBeginOverlap
  (params (OtherActor :type Actor%20Object%20Reference))
  (|CallParentFunction OtherActor) ; This node is disabled and will not be called.\nDrag off pins to build functionality.
)

(event EventTick
  (params (DeltaSeconds :type Float%20%28single-precision%29))
  (|CallParentFunction DeltaSeconds) ; This node is disabled and will not be called.\nDrag off pins to build functionality.
)

(fn ConstructionScript
  (bind _staticmeshcomponent (Variables|Debug|GetStaticMeshComponent))
  (|CallParentFunction)
  (bind _outtexturedataarray (Building|GetBuildingTextureData))
  (if (> (Utilities|Array|Length _outtexturedataarray) 0)
    (bind _outtexturedataarray_2 (Building|GetBuildingTextureData))
    (Default|SetColoronSecondaryMesh (Math|Conversions|ToLinearColor(Color) (Variables|/Script/FortniteGame.BuildingTextureData|GetTintColor (Utilities|Array|Get(acopy) _outtexturedataarray_2))) _staticmeshcomponent)
    (Rendering|Material|SetMaterial _staticmeshcomponent 0 (Class|BTechArtGlobalLibrary|GetDynamicMaterialInstance (Variables|Debug|GetStaticMeshComponent)))))
